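# HUD element wrapper, derived from FortMobileHUDElement, that hosts the Meatball (motorboat) speedometer widget and attaches it to the local player's vehicle.
# On Construct it subscribes to the player controller's driver-enter, full-exit and seat-change delegates and tries an immediate bind; Destruct removes the subscriptions.
# Driver entry, or a change to seat 0, caches the pawn's vehicle and passes it to the speedometer via BindTo.
# Full exit calls UnbindFrom and clears the cache.
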WidgetBlueprint /Motorboat/Vehicle/HUD/Widgets/WBP_MeatballSpeedometer_Wrapper
graphs [EventGraph {nodes [
   OnEnterVehicleDriver() {is=title K2Node_CustomEvent}
   > n1 = Cast To FortPlayerController(Object=GetOwningPlayer())
   > n2 = Cast To FortPlayerPawnAthena(Object=GetPlayerPawn!(self:Fort Player Controller Object Reference=@n1) -> (ReturnValue:Fort Player Pawn Object Reference))
   n3 = Get Vehicle!(self:Fort Player Pawn Object Reference=@n2) -> (ReturnValue:Fort Athena Vehicle Object Reference) {is=title K2Node_CallFunction}
   > Is Valid(InputObject=@n3)
   > n4 = Set CachedVehicle(CachedVehicle:Fort Athena Vehicle Object Reference=@n3) -> (Output_Get:Fort Athena Vehicle Object Reference)
   > BindTo!(
    execute:Exec
    self:Object Reference=Get WBP_MeatballSpeedoMeter() -> (WBP_MeatballSpeedoMeter:WBP Meatball Speedo Meter Object Reference)
    AssociatedObject:Object Reference=@n4
   ) -> (then:Exec) {is=title K2Node_Message}
   OnFullyExitVehicle() {is=title K2Node_CustomEvent}
   > n5 = Get!(execute:Exec) -> (then:Exec, else:Exec, CachedVehicle:Fort Athena Vehicle Object Reference)
   > UnbindFrom!(execute:Exec, self:Object Reference=Get WBP_MeatballSpeedoMeter(), AssociatedObject:Object Reference=@n5) -> (then:Exec) {is=title K2Node_Message}
   > Set CachedVehicle()
   n6 = OnSeatChanged!() -> (OutputDelegate, then, SeatIndex:int) {is=title K2Node_CustomEvent}
   > Branch(Condition="Equal (Integer)"(A=@n6.SeatIndex))
   > OnEnterVehicleDriver() {is=title K2Node_CallFunction}
   Event Construct()
   > n7 = Cast To FortPlayerController(Object=GetOwningPlayer())
   > n8 = Set ChachedPlayerController(ChachedPlayerController:Fort Player Controller Object Reference=@n7) -> (Output_Get:Fort Player Controller Object Reference)
   > Bind Event to On Enter Vehicle Driver(self:Fort Player Controller Object Reference=@n8, Delegate=Create Event()) {is=title K2Node_AddDelegate}
   > Assign On Fully Exit Vehicle!(
    execute:Exec
    self:Fort Player Controller Object Reference=Get ChachedPlayerController() -> (ChachedPlayerController:Fort Player Controller Object Reference)
    Delegate:"Delegate (by ref)"=Create Event()
   ) -> (then:Exec) {is=title K2Node_AssignDelegate}
   > Bind Event to On Vehicle Seat Changed(self:Fort Player Controller Object Reference=Get ChachedPlayerController(), Delegate=Create Event()) {is=title K2Node_AddDelegate}
   > OnEnterVehicleDriver()
   Event Destruct()
   > n9 = Get!(execute) -> (then, else, ChachedPlayerController:Fort Player Controller Object Reference)
   > Unbind Event from On Enter Vehicle Driver(self:Fort Player Controller Object Reference=@n9, Delegate=Create Event()) {is=title K2Node_RemoveDelegate}
   > Unbind Event from On Fully Exit Vehicle(self:Fort Player Controller Object Reference=Get ChachedPlayerController(), Delegate=Create Event()) {is=title K2Node_RemoveDelegate}
   > Unbind Event from On Vehicle Seat Changed(self:Fort Player Controller Object Reference=Get ChachedPlayerController(), Delegate=Create Event()) {is=title K2Node_RemoveDelegate}
  ]}]
variables [
 CachedVehicle {type=FortAthenaVehicle, default=null}
 ChachedPlayerController {type=FortPlayerController, default=null}
]
widgets {
 info {inheritedWidgetCount=3, parentClass=/Script/FortniteUI.FortMobileHUDElement, rootWidgetClass=&ScaleBox, widgetCount=4}
 widgets [
  ScaleBox_HUDWidgetScale {widgetClassPath=&ScaleBox, bInherited=true, widget=ScaleBox_HUDWidgetScale, properties {stretch=UserSpecified}}
  WBP_MeatballSpeedoMeter {
   widgetClassPath=&/WBP_MeatballSpeedoMeter
   bIsVariable=true
   parent=ScaleBox_HUDWidgetScale
   widget=WBP_MeatballSpeedoMeter
   slot {class=ScaleBoxSlot}
  }
  Image_Background {widgetClassPath=&Image, bInherited=true, widget=null}
  TutorialHighlight {widgetClassPath=&TutorialHighlight}
 ]
}
